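# Material function for the StrideMice player body cosmetic that adds an audio-reactive emissive glow to incoming material attributes.
# It takes the maximum of a Fresnel rim and an emissive mask (switchable texture source), then tints it with two blended colors and scales brightness.
# Named reroutes supply powered audio amplitude from MF_AudioReactivity_FrontendLoop, panning noise, an upper-body texture mask and a vertex-interpolated pre-skinning sphere mask.
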MaterialFunction /BRCosmetics/Characters/Player/Male/Medium/Bodies/M_MED_StrideMiceDeep_Canny/Materials/MF_StrideMIce_Reactive
graph {nodes [
  n1 = GetMaterialAttributes(None=FunctionInput() {inputType=FunctionInput_MaterialAttributes}) {attributeGetTypes [#1, #2]}
  n2 = NamedRerouteUsage()
  FunctionOutput(None=SetMaterialAttributes!(MaterialAttributes=@n1.MaterialAttributes, EmissiveColor=Add(A=@n1.EmissiveColor, B=@n2)) {attributeSetTypes [#1]})
  n3 = Fresnel(ExponentIn=LinearInterpolate(
    A=ScalarParameter() {defaultValue=3, group=StideMIce_Reactive, parameterName=EmissiveFresnelExponent, sortPriority=90}
    B=ScalarParameter() {defaultValue=2.5, group=StideMIce_Reactive, parameterName=EmissiveFresnelExponent_Reactive, sortPriority=90}
    Alpha=NamedRerouteUsage()
   ) {constB=2}) {baseReflectFraction=0}
  n4 = ScalarParameter() {defaultValue=1, group=StideMIce_Reactive, parameterName=EmissiveFresnelIntensity, sortPriority=92}
  n5 = ComponentMask(None=NamedRerouteUsage()) {g=true}
  n6 = NamedRerouteDeclaration(None=Multiply(A=Multiply(A=@n3, B=@n4), B=@n5)) {name=FresnelMask, nodeColor=<rgba 0 0.7891 1 1>}
  n7 = TextureSampleParameter2D() {group=StideMIce_Reactive, parameterName=EmissiveMap, samplerType=SAMPLERTYPE_Masks, texture=&T_BlackMask}
  n8 = Power(Base=@n7.R, Exp=ScalarParameter() {defaultValue=2, group=StideMIce_Reactive, parameterName=EmissiveMapExponent})
  n9 = TextureSampleParameter2D() {
   group=00_Texture Inputs
   parameterName=Mask
   samplerSource=SSM_Wrap_WorldGroupSettings
   samplerType=SAMPLERTYPE_Masks
   sortPriority=3
   texture=&T_BlackMask
  }
  n10 = Saturate(None=Power(
    Base=OneMinus(None=@n9.G)
    Exp=ScalarParameter() {defaultValue=3, group=StideMIce_Reactive, parameterName=EmissiveMaskExponent, sortPriority=93}
   ))
  n11 = Multiply(
   A=Multiply(A=@n5, B=StaticSwitchParameter(True=@n8, False=@n10) {group=StideMIce_Reactive, parameterName=UseEmissiveMap, sortPriority=89})
   B=ScalarParameter() {defaultValue=0.8, group=StideMIce_Reactive, parameterName=EmissiveMaskIntensity, sortPriority=92}
  )
  n12 = VectorParameter() {group=StideMIce_Reactive, parameterName=EmissiveColor_ReactiveA, sortPriority=91, defaultValue=<rgba 0.956 0.6757 0.239 1>}
  n13 = VectorParameter() {group=StideMIce_Reactive, parameterName=EmissiveColor_ReactiveB, sortPriority=91, defaultValue=<rgba 0.6 0.4762 0.18 1>}
  n14 = NamedRerouteUsage()
  n15 = Multiply(
   A=Multiply(A=Max(A=@n6, B=@n11), B=LinearInterpolate(A=@n12.RGB, B=@n13.RGB, Alpha=@n14))
   B=LinearInterpolate(B=ScalarParameter() {defaultValue=15, group=StideMIce_Reactive, parameterName=EmissiveBrightness_Reactive}, Alpha=NamedRerouteUsage()) {constB=3}
  )
  n16 = NamedRerouteUsage()
  NamedRerouteDeclaration(None=LinearInterpolate(A=@n15, B=Multiply(A=@n15, B=@n16), Alpha=ScalarParameter() {group=StideMIce_Reactive, parameterName=upper body Mask}) {constAlpha=0}) {name=Fresnel, nodeColor=<rgba 0.8047 0 1 1>}
  n17 = VertexInterpolator(VS=LocalPosition() {localOrigin=InstancePreSkinning}.XYZ)
  n18 = VectorParameter() {group=StideMIce_Reactive, parameterName=Reactive_MaskCenter, defaultValue=<rgba 0 0 141.908 1>}
  n19 = ScalarParameter() {defaultValue=120, group=StideMIce_Reactive, parameterName=Reactive_Radius}
  n20 = ScalarParameter() {defaultValue=-0.68, group=StideMIce_Reactive, parameterName=Reactive_Hardness}
  NamedRerouteDeclaration(None=Power(
    Base=SphereMask(A=@n17.PS, B=@n18.RGB, Radius=@n19, Hardness=@n20)
    Exp=ScalarParameter() {defaultValue=2.2, group=StideMIce_Reactive, parameterName=Reactive_Radius_power}
   ) {constExponent=3}) {name=FresnelPreSkinMask, nodeColor=<rgba 0 0.625 1 1>}
  NamedRerouteDeclaration(None=@n7.G) {name=UpperBodyMask, nodeColor=<rgba 1 0 0.2812 1>}
  n21 = TextureCoordinate()
  n22 = Constant() {r=3}
  n23 = TextureSample(UVs=Panner(Coordinate=Multiply(A=@n21, B=@n22)) {speedX=0.1, speedY=0.1}) {samplerSource=SSM_Wrap_WorldGroupSettings, samplerType=SAMPLERTYPE_Grayscale, texture=&LowResBlurredNoise}
  NamedRerouteDeclaration(None=Power(Base=@n23.R)) {name=ColorNoise, nodeColor=<rgba 0.1484 0 1 1>}
  n24 = MaterialFunctionCall!(InAmplitude, InSpectrum, CustomFakeSpectrum, CustomFakeAmplitude) {materialFunction=/BRCosmetics/Global/MaterialFunctions/AudioReactivity/MF_AudioReactivity_FrontendLoop}
  n25 = ScalarParameter() {defaultValue=1, group=StideMIce_Reactive, parameterName=EnableAudioReactivity, sliderMax=1}
  NamedRerouteDeclaration(None=Power(
    Base=Multiply(A=@n24.Amplitude, B=@n25)
    Exp=ScalarParameter() {defaultValue=3, group=StideMIce_Reactive, parameterName=AudioReactivityPower}
   ) {constExponent=5}) {name=Amplitude, nodeColor=<rgba 0 1 0.1094 1>}
 ]}
