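# Physics asset for the M_MED_StrideMiceDeep_Canny male body cosmetic, set up for a dangling necklace on the upper chest.
# A two-link chain, dyn_necklace and dyn_necklace_01, hangs from base_necklace with uneven swing limits and zero twist range.
# Capsules on spine_05, both clavicles and a dedicated necklace_collider bone form the surrounding body collision.
# The base, clavicles and collider each constrain to spine_05 with free swing and twist.
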PhysicsAsset /BRCosmetics/Characters/Player/Male/Medium/Bodies/M_MED_StrideMiceDeep_Canny/Meshes/M_MED_StrideMiceDeep_Canny_Physics
bodies {
 base_necklace [{shapeType=Capsule, length=0.505, radius=0.505, shapeName=base_necklace_capsule}]
 clavicle_l [{shapeType=Capsule, length=9.9592, radius=8.3758, shapeName=clavicle_l_capsule, center=<6.967 0 0>, rotation=<rot 90 0 0>}]
 clavicle_r [{shapeType=Capsule, length=10.0357, radius=7.6101, shapeName=clavicle_r_capsule, center=<-8.1797 0 0>, rotation=<rot 90 0 0>}]
 dyn_necklace [{shapeType=Capsule, length=2, radius=1.5, shapeName=dyn_necklace_capsule, center=<1.6026 0.2337 0>, rotation=<rot 90 -5.6254 0>}]
 dyn_necklace_01 [{
   shapeType=Capsule
   length=1.7008
   radius=1.4089
   shapeName=dyn_necklace_01_capsule
   center=<1.5629 -0.4486 0.1009>
   rotation=<rot -0.012 -99.3432 -89.8786>
  }]
 necklace_collider [{shapeType=Capsule, length=4, radius=4, shapeName=necklace_collider_capsule, center=<1.8084 2.7635 0>, rotation=<rot 0 -90 0>}]
 spine_05 [
  {shapeType=Capsule, length=12.7626, radius=11.2669, shapeName=spine_05_capsule, center=<-1.2892 -1.4128 0>}
  {shapeType=Capsule, length=15, radius=15, shapeName=spine_05_capsule, center=<-6.0185 -2.6978 0>}
 ]
}
constraints [
 {
  bone1Name=base_necklace
  bone2Name=spine_05
  swing1LimitDegrees=45
  swing1Motion=Free
  swing2LimitDegrees=45
  swing2Motion=Free
  twistLimitDegrees=45
  twistMotion=Free
 }
 {bone1Name=dyn_necklace, bone2Name=base_necklace, swing1LimitDegrees=60, swing2LimitDegrees=10, twistLimitDegrees=0}
 {bone1Name=dyn_necklace_01, bone2Name=dyn_necklace, swing1LimitDegrees=60, swing2LimitDegrees=10, twistLimitDegrees=0}
 {
  bone1Name=clavicle_l
  bone2Name=spine_05
  swing1LimitDegrees=45
  swing1Motion=Free
  swing2LimitDegrees=45
  swing2Motion=Free
  twistLimitDegrees=45
  twistMotion=Free
 }
 {
  bone1Name=clavicle_r
  bone2Name=spine_05
  swing1LimitDegrees=45
  swing1Motion=Free
  swing2LimitDegrees=45
  swing2Motion=Free
  twistLimitDegrees=45
  twistMotion=Free
 }
 {
  bone1Name=necklace_collider
  bone2Name=spine_05
  swing1LimitDegrees=45
  swing1Motion=Free
  swing2LimitDegrees=45
  swing2Motion=Free
  twistLimitDegrees=45
  twistMotion=Free
 }
]
