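# Post-process style AnimBlueprint for the StrideMiceDeep_Canny male character body, adding secondary motion to shirt and necklace over the attached parent's copied pose.
# A Control Rig driven by parachute, crouch, targeting and emote values feeds RigidBody, shirt AnimDynamics, a targeting-weighted necklace constraint, then necklace dynamics chosen by front-end state.
# The EventGraph updates those values and state-based gravity, with skydive wind, for the RigidBody and shirt.
# Several alternative physics nodes are unconnected to the output.
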AnimBlueprint /BRCosmetics/Characters/Player/Male/Medium/Bodies/M_MED_StrideMiceDeep_Canny/Meshes/M_MED_StrideMiceDeep_Canny_AnimBP
target_skeleton=&/M_MED_StrideMiceDeep_Canny_Skeleton
graphs [
 AnimGraph {nodes [
   n1 = Anim Dynamics - Bone: dyn_shirt_front_r!(ComponentPose, Alpha=1, GravityOverride:Vector="0,0,-980", ExternalForce="0,0,0") -> (Pose) {
    angularDampingOverride=1
    angularSpringConstant=5000
    bAngularSpring=true
    bOverrideAngularDamping=true
    bUseGravityOverride=true
    gravityScale=0.35
    simulationSpace=RootRelative
    alphaBoolBlend {blendOption=Linear}
    boundBone {boneName=dyn_shirt_front_r}
    componentLinearAccScale=<0.5 0.5 0.5>
    componentLinearVelScale=<0.5 0.5 0.5>
    gravityOverride {z=-980}
    physicsBodyDefinitions [{
      boundBone {boneName=dyn_shirt_front_r}
      boxExtents=<20 10 10>
      constraintSetup {angularLimitsMax=<15 0 15>, angularLimitsMin=<-15 -30 -15>, angularTarget=<1 0 0>}
      localJointOffset=<5 0 0>
     }]
    simSpaceSettings {simSpaceAngularAlpha=0.5}
   }
   n2 = Anim Dynamics - Bone: dyn_shirt_front_l!(ComponentPose=@n1, Alpha=1, GravityOverride:Vector="0,0,-980", ExternalForce="0,0,0") -> (Pose) {
    angularDampingOverride=1
    angularSpringConstant=5000
    bAngularSpring=true
    bOverrideAngularDamping=true
    bUseGravityOverride=true
    gravityScale=0.35
    simulationSpace=RootRelative
    alphaBoolBlend {blendOption=Linear}
    boundBone {boneName=dyn_shirt_front_l}
    componentLinearAccScale=<0.5 0.5 0.5>
    componentLinearVelScale=<0.5 0.5 0.5>
    gravityOverride {z=-980}
    physicsBodyDefinitions [{
      boundBone {boneName=dyn_shirt_front_l}
      boxExtents=<20 10 10>
      constraintSetup {angularLimitsMax=<15 0 15>, angularLimitsMin=<-15 -30 -15>, angularTarget=<1 0 0>}
      localJointOffset=<5 0 0>
     }]
    simSpaceSettings {simSpaceAngularAlpha=0.5}
   }
   n3 = Anim Dynamics - Bone: dyn_shirt_back!(ComponentPose=@n2, Alpha=1, GravityOverride:Vector="0,0,-980", ExternalForce="0,0,0") -> (Pose) {
    angularDampingOverride=1
    angularSpringConstant=5000
    bAngularSpring=true
    bOverrideAngularDamping=true
    bUseGravityOverride=true
    gravityScale=0.35
    simulationSpace=RootRelative
    alphaBoolBlend {blendOption=Linear}
    boundBone {boneName=dyn_shirt_back}
    componentLinearAccScale=<0.5 0.5 0.5>
    componentLinearVelScale=<0.5 0.5 0.5>
    gravityOverride {z=-980}
    physicsBodyDefinitions [{
      boundBone {boneName=dyn_shirt_back}
      boxExtents=<20 10 10>
      constraintSetup {angularLimitsMax=<15 0 20>, angularLimitsMin=<-15 -30 -20>, angularTarget=<1 0 0>}
      localJointOffset=<5 0 0>
     }]
    simSpaceSettings {simSpaceAngularAlpha=0.5}
   }
   n4 = Anim Dynamics - Bone: base_shirt_l!(ComponentPose=@n3, Alpha=1, GravityOverride:Vector="0,0,-980", ExternalForce="0,0,0") -> (Pose) {
    angularDampingOverride=1
    angularSpringConstant=5000
    bAngularSpring=true
    bOverrideAngularDamping=true
    bUseGravityOverride=true
    gravityScale=0.35
    simulationSpace=RootRelative
    alphaBoolBlend {blendOption=Linear}
    boundBone {boneName=base_shirt_l}
    componentLinearAccScale=<0.5 0.5 0.5>
    componentLinearVelScale=<0.5 0.5 0.5>
    gravityOverride {z=-980}
    physicsBodyDefinitions [{
      boundBone {boneName=base_shirt_l}
      boxExtents=<20 10 10>
      constraintSetup {angularLimitsMax=<15 0 15>, angularLimitsMin=<-15 -30 -15>, angularTarget=<1 0 0>}
      localJointOffset=<5 0 0>
     }]
    simSpaceSettings {simSpaceAngularAlpha=0.5}
   }
   Anim Dynamics - Bone: dyn_shirt_r!(ComponentPose=@n4, Alpha=1, GravityOverride:Vector="0,0,-980", ExternalForce="0,0,0") -> (Pose) {
    angularDampingOverride=1
    angularSpringConstant=5000
    bAngularSpring=true
    bOverrideAngularDamping=true
    bUseGravityOverride=true
    gravityScale=0.35
    simulationSpace=RootRelative
    alphaBoolBlend {blendOption=Linear}
    boundBone {boneName=dyn_shirt_r}
    componentLinearAccScale=<0.5 0.5 0.5>
    componentLinearVelScale=<0.5 0.5 0.5>
    gravityOverride {z=-980}
    physicsBodyDefinitions [{
      boundBone {boneName=dyn_shirt_r}
      boxExtents=<20 10 10>
      constraintSetup {angularLimitsMax=<15 0 15>, angularLimitsMin=<-15 -30 -15>, angularTarget=<1 0 0>}
      localJointOffset=<5 0 0>
     }]
    simSpaceSettings {simSpaceAngularAlpha=0.5}
   }
   Anim Dynamics - Bone: dyn_shirt_back(Alpha=1, GravityOverride="0,0,-980", ExternalForce="0,0,0") {
    angularDampingOverride=1
    angularSpringConstant=20000
    bAngularSpring=true
    bOverrideAngularDamping=true
    bUseGravityOverride=true
    gravityScale=0.35
    alphaBoolBlend {blendOption=Linear}
    boundBone {boneName=dyn_shirt_back}
    componentLinearAccScale=<1 1 1>
    componentLinearVelScale=<1 1 1>
    gravityOverride {z=-980}
    physicsBodyDefinitions [{
      boundBone {boneName=dyn_shirt_back}
      boxExtents=<20 30 30>
      constraintSetup {angularLimitsMax=<0 60 0>, angularLimitsMin=<0 0 0>, angularTarget=<1 0 0>}
      localJointOffset=<4 0 0>
     }]
   }
   n5 = RigidBody!(
    ComponentPose=Local To Component(LocalPose=Use cached pose 'start'())
    Alpha
    OverrideWorldGravity:Vector=Get GravityOverride()
    ExternalForce
   ) -> (Pose) {
    bClampLinearTranslationLimitToRefPose=true
    bOverrideWorldGravity=true
    overridePhysicsAsset=&/M_MED_StrideMiceDeep_Canny_Physics_f
    simulationSpace=BaseBoneSpace
    alphaBoolBlend {blendOption=Linear}
    alphaScaleBiasClamp {bInterpResult=true, bias=5, scale=5}
    baseBoneRef {boneName=pelvis}
    componentAppliedLinearAccClamp=<3000 3000 3000>
    componentLinearAccScale=<0.3 0.15 0.3>
    componentLinearVelScale=<0.3 0.15 0.3>
    overrideWorldGravity {z=-980}
    simSpaceSettings {
     maxAngularAcceleration=250
     maxAngularVelocity=50
     maxLinearAcceleration=400
     maxLinearVelocity=100
     velocityScaleZ=0
     worldAlpha=0.3
     externalLinearDragV=<0.2 0.2 0.1>
    }
   }
   Blend Poses by bool(BlendPose_0=Component To Local(ComponentPose=@n5), BlendPose_1=Component To Local(), bActiveValue=Get MoveNOCrouch() -> (MoveNOCrouch:bool)) {blendType=Linear}
   Copy Pose From Mesh() {bUseAttachedParent=true}
   "Transform (Modify) Bone - Bone: dyn_necklace_01"(Alpha=1, Translation="0,0,0", Rotation="0,20,0", Scale="1,1,1") {
    rotationMode=BMM_Additive
    rotationSpace=BCS_BoneSpace
    translationMode=BMM_Additive
    translationSpace=BCS_BoneSpace
    alphaBoolBlend {blendOption=Linear}
    boneToModify {boneName=dyn_necklace_01}
    rotation {yaw=20}
   }
   "Transform (Modify) Bone - Bone: base_shirt"(Alpha=1, Translation="0,100,0", Rotation="0,0,0", Scale="1,1,1") {
    translationMode=BMM_Additive
    translationSpace=BCS_BoneSpace
    alphaBoolBlend {blendOption=Linear}
    boneToModify {boneName=base_shirt}
    translation {y=100}
   }
   "Transform (Modify) Bone - Bone: base_necklace"(Alpha=1, Translation="0,-1,0", Rotation="0,6,0", Scale="1,1,1") {
    rotationMode=BMM_Additive
    rotationSpace=BCS_BoneSpace
    translationMode=BMM_Additive
    translationSpace=BCS_BoneSpace
    alphaBoolBlend {blendOption=Linear}
    boneToModify {boneName=base_necklace}
    rotation {yaw=6}
    translation {y=-1}
   }
   RigidBody(OverrideWorldGravity="0,0,0") {
    bClampLinearTranslationLimitToRefPose=true
    bOverrideWorldGravity=true
    overridePhysicsAsset=&/M_MED_StrideMiceDeep_Canny_Physics
    simulationSpace=BaseBoneSpace
    alphaBoolBlend {blendOption=Linear}
    baseBoneRef {boneName=pelvis}
    componentAppliedLinearAccClamp=<3000 3000 3000>
    componentLinearAccScale=<0.3 0.15 0.3>
    componentLinearVelScale=<0.3 0.15 0.3>
    overrideWorldGravity {z=-980}
    simSpaceSettings {
     maxAngularAcceleration=500
     maxAngularVelocity=100
     maxLinearAcceleration=200
     maxLinearVelocity=100
     velocityScaleZ=0
     worldAlpha=1
     externalLinearDragV=<0.2 0.2 0.2>
    }
   }
   RigidBody(OverrideWorldGravity=Get GravityOverride()) {
    bClampLinearTranslationLimitToRefPose=true
    bOverrideWorldGravity=true
    overridePhysicsAsset=&M_MED_StrideMiceDeep_Canny_Physics
    simulationSpace=BaseBoneSpace
    alphaBoolBlend {blendOption=Linear}
    alphaScaleBiasClamp {bInterpResult=true, bias=5, scale=5}
    baseBoneRef {boneName=pelvis}
    componentAppliedLinearAccClamp=<3000 3000 3000>
    componentLinearAccScale=<0.3 0.15 0.3>
    componentLinearVelScale=<0.3 0.15 0.3>
    overrideWorldGravity {z=-980}
    simSpaceSettings {
     maxAngularAcceleration=250
     maxAngularVelocity=50
     maxLinearAcceleration=400
     maxLinearVelocity=100
     velocityScaleZ=0
     worldAlpha=0.3
     externalLinearDragV=<0.2 0.2 0.1>
    }
   }
   RigidBody(OverrideWorldGravity="0,0,0") {
    bClampLinearTranslationLimitToRefPose=true
    bOverrideWorldGravity=true
    overridePhysicsAsset=&M_MED_StrideMiceDeep_Canny_Physics
    simulationSpace=BaseBoneSpace
    alphaBoolBlend {blendOption=Linear}
    baseBoneRef {boneName=pelvis}
    componentAppliedLinearAccClamp=<3000 3000 3000>
    componentLinearAccScale=<0.3 0.15 0.3>
    componentLinearVelScale=<0.3 0.15 0.3>
    overrideWorldGravity {z=-980}
    simSpaceSettings {
     maxAngularAcceleration=250
     maxAngularVelocity=500
     maxLinearAcceleration=400
     maxLinearVelocity=100
     velocityScaleZ=0
     worldAlpha=0.3
     externalLinearDragV=<0.2 0.2 0.1>
    }
   }
   RigidBody(OverrideWorldGravity="0,0,0") {
    bOverrideWorldGravity=true
    overridePhysicsAsset=/Game/Characters/Player/Female/Medium/Bodies/F_MED_RockerPunkAlt/Meshes/F_MED_RockerPunkAlt_PhysicsRunning
    simulationSpace=BaseBoneSpace
    alphaBoolBlend {blendOption=Linear}
    baseBoneRef {boneName=pelvis}
    componentAppliedLinearAccClamp=<3000 3000 3000>
    componentLinearAccScale=<0.3 0.3 0.4>
    componentLinearVelScale=<0.3 0.3 0.4>
    simSpaceSettings {maxAngularVelocity=3, maxLinearAcceleration=700, maxLinearVelocity=500, worldAlpha=0.6, externalLinearDragV=<0.2 0.2>}
   }
   n6 = Anim Dynamics - Bone: dyn_necklace!(ComponentPose=Local To Component(LocalPose=Use cached pose 'Necklace'()), Alpha=1, GravityOverride:Vector="0,0,-980", ExternalForce="0,0,0") -> (Pose) {
    angularDampingOverride=0.999
    angularSpringConstant=500
    bAngularSpring=true
    bOverrideAngularDamping=true
    bUseGravityOverride=true
    gravityScale=0.35
    simulationSpace=RootRelative
    alphaBoolBlend {blendOption=Linear}
    boundBone {boneName=dyn_necklace}
    componentAppliedLinearAccClamp=<3000 3000 3000>
    componentLinearAccScale=<0.5 0.5 0.5>
    componentLinearVelScale=<0.5 0.5 0.5>
    gravityOverride {z=-980}
    physicsBodyDefinitions [{
      boundBone {boneName=dyn_necklace}
      boxExtents=<30 15 15>
      constraintSetup {angularLimitsMax=<15 10 0>, angularLimitsMin=<-15 -10 -90>, angularTarget=<1 0 0>}
      localJointOffset=<1 0 0>
     }]
    relativeSpaceBone {boneName=spine_05}
    simSpaceSettings {maxAngularVelocity=100, simSpaceAngularAlpha=0.5, externalAngularVelocity=<1 1 1>}
   }
   n7 = Anim Dynamics - Bone: dyn_necklace_01!(ComponentPose=@n6, Alpha=1, GravityOverride:Vector="0,0,-980", ExternalForce="0,0,0") -> (Pose) {
    angularDampingOverride=0.999
    angularSpringConstant=500
    bAngularSpring=true
    bOverrideAngularDamping=true
    bUseGravityOverride=true
    gravityScale=0.35
    simulationSpace=RootRelative
    alphaBoolBlend {blendOption=Linear}
    boundBone {boneName=dyn_necklace_01}
    componentAppliedLinearAccClamp=<3000 3000 3000>
    componentLinearAccScale=<0.2 0.2 0.2>
    componentLinearVelScale=<0.2 0.2 0.2>
    gravityOverride {z=-980}
    physicsBodyDefinitions [{
      boundBone {boneName=dyn_necklace_01}
      boxExtents=<30 15 15>
      constraintSetup {angularLimitsMax=<15 30 0>, angularLimitsMin=<-15 -30 -10>, angularTarget=<1 0 0>}
      localJointOffset=<1 0 0>
     }]
    relativeSpaceBone {boneName=base_necklace}
    simSpaceSettings {maxAngularVelocity=100, simSpaceAngularAlpha=0.5}
   }
   n8 = Anim Dynamics - Bone: dyn_necklace(ComponentPose=Local To Component(LocalPose=Use cached pose 'Necklace'()), Alpha=1, GravityOverride="0,0,-980", ExternalForce="0,0,0") {
    angularDampingOverride=0.999
    angularSpringConstant=5000
    bAngularSpring=true
    bOverrideAngularDamping=true
    bUseGravityOverride=true
    gravityScale=0.35
    simulationSpace=RootRelative
    alphaBoolBlend {blendOption=Linear}
    boundBone {boneName=dyn_necklace}
    componentAppliedLinearAccClamp=<3000 3000 3000>
    componentLinearAccScale=<0.5 0.5 0.5>
    componentLinearVelScale=<0.5 0.5 0.5>
    gravityOverride {z=-980}
    physicsBodyDefinitions [{
      boundBone {boneName=dyn_necklace}
      boxExtents=<30 15 15>
      constraintSetup {angularLimitsMax=<15 15 0>, angularLimitsMin=<-15 -15 -90>, angularTarget=<1 0 0>}
      localJointOffset=<5 0 0>
     }]
    relativeSpaceBone {boneName=spine_05}
    simSpaceSettings {maxAngularVelocity=100, simSpaceAngularAlpha=0.5, externalAngularVelocity=<1 1 1>}
   }
   n9 = Anim Dynamics - Bone: dyn_necklace_01(ComponentPose=@n8, Alpha=1, GravityOverride="0,0,-980", ExternalForce="0,0,0") {
    angularDampingOverride=0.999
    angularSpringConstant=5000
    bAngularSpring=true
    bOverrideAngularDamping=true
    bUseGravityOverride=true
    gravityScale=0.35
    simulationSpace=RootRelative
    alphaBoolBlend {blendOption=Linear}
    boundBone {boneName=dyn_necklace_01}
    componentAppliedLinearAccClamp=<3000 3000 3000>
    componentLinearAccScale=<0.2 0.2 0.2>
    componentLinearVelScale=<0.2 0.2 0.2>
    gravityOverride {z=-980}
    physicsBodyDefinitions [{
      boundBone {boneName=dyn_necklace_01}
      boxExtents=<30 15 15>
      constraintSetup {angularLimitsMax=<15 40 0>, angularLimitsMin=<-15 -40 -10>, angularTarget=<1 0 0>}
      localJointOffset=<2 0 0>
     }]
    relativeSpaceBone {boneName=base_necklace}
    simSpaceSettings {maxAngularVelocity=100, simSpaceAngularAlpha=0.5}
   }
   n10 = Blend Poses by bool(BlendPose_0=Component To Local(ComponentPose=@n7), BlendPose_1=Component To Local(ComponentPose=@n9), bActiveValue=Get bIsInFrontEnd()) {blendType=Linear}
   Output Pose(Result=Component To Local(ComponentPose=Local To Component(LocalPose=@n10))) {layerGroup=None}
   n11 = Control Rig!(
    Source=Copy Pose From Mesh() {bUseAttachedParent=true}
    Alpha
    parachute:double=Get Parachute() -> (Parachute:double)
    crouch:double=Get crouching() -> (crouching:double)
    Target:double=Get Target() -> (Target:double)
    Emote:double=Get Emote() -> (Emote:double)
   ) -> (Pose) {
    alphaBoolBlend {blendOption=Linear}
    alphaScaleBiasClamp {bInterpResult=true}
    controlRigAssetReference {blueprintRigClass=&/M_MED_StrideMiceDeep_Canny_CtrlRig.M_MED_StrideMiceDeep_Canny_CtrlRig_C}
   }
   Save cached pose 'start'(Pose=@n11) {cacheName=start}
   n12 = Get GravityOverride()
   n13 = RigidBody(ComponentPose=Local To Component(LocalPose=Use cached pose 'start'()), OverrideWorldGravity=@n12) {
    bClampLinearTranslationLimitToRefPose=true
    bOverrideWorldGravity=true
    overridePhysicsAsset=&/M_MED_StrideMiceDeep_Canny_Back_Physics
    simulationSpace=BaseBoneSpace
    alphaBoolBlend {blendOption=Linear}
    alphaScaleBiasClamp {bInterpResult=true, bias=5, scale=5}
    baseBoneRef {boneName=pelvis}
    componentAppliedLinearAccClamp=<3000 3000 3000>
    componentLinearAccScale=<0.3 0.15 0.3>
    componentLinearVelScale=<0.3 0.15 0.3>
    overrideWorldGravity {z=-980}
    simSpaceSettings {
     maxAngularAcceleration=250
     maxAngularVelocity=50
     maxLinearAcceleration=400
     maxLinearVelocity=100
     velocityScaleZ=0
     worldAlpha=0.3
     externalLinearDragV=<0.2 0.2 0.1>
    }
   }
   n14 = Anim Dynamics - Bone: dyn_shirt_r(ComponentPose=@n13, Alpha=1, GravityOverride=@n12, ExternalForce="0,0,0") {
    angularDampingOverride=0.99
    angularSpringConstant=1000
    bAngularSpring=true
    bOverrideAngularDamping=true
    bUseGravityOverride=true
    gravityScale=0.35
    alphaBoolBlend {blendOption=Linear}
    boundBone {boneName=dyn_shirt_r}
    componentAppliedLinearAccClamp=<6000 6000 6000>
    componentLinearAccScale=<0.6 0.6 0.6>
    componentLinearVelScale=<0.6 0.6 0.6>
    gravityOverride {z=-980}
    physicsBodyDefinitions [{
      boundBone {boneName=dyn_shirt_r}
      boxExtents=<20 20 20>
      constraintSetup {angularLimitsMax=<0 0 0>, angularLimitsMin=<0 -50 0>, angularTarget=<1 0 0>}
      localJointOffset=<5 0 0>
     }]
   }
   n15 = Anim Dynamics - Bone: base_shirt_l(ComponentPose=@n14, Alpha=1, GravityOverride=@n12, ExternalForce="0,0,0") {
    angularDampingOverride=0.99
    angularSpringConstant=1000
    bAngularSpring=true
    bOverrideAngularDamping=true
    bUseGravityOverride=true
    gravityScale=0.35
    alphaBoolBlend {blendOption=Linear}
    boundBone {boneName=base_shirt_l}
    componentAppliedLinearAccClamp=<6000 6000 6000>
    componentLinearAccScale=<0.6 0.6 0.6>
    componentLinearVelScale=<0.6 0.6 0.6>
    gravityOverride {z=-980}
    physicsBodyDefinitions [{
      boundBone {boneName=base_shirt_l}
      boxExtents=<20 20 20>
      constraintSetup {angularLimitsMax=<0 30 5>, angularLimitsMin=<0 0 -5>, angularTarget=<1 0 0>}
      localJointOffset=<5 0 0>
     }]
   }
   n16 = Anim Dynamics - Bone: dyn_shirt_front_r(ComponentPose=@n15, Alpha=1, GravityOverride=@n12, ExternalForce="0,0,0") {
    angularDampingOverride=0.99
    angularSpringConstant=1000
    bAngularSpring=true
    bOverrideAngularDamping=true
    bUseGravityOverride=true
    gravityScale=0.35
    alphaBoolBlend {blendOption=Linear}
    boundBone {boneName=dyn_shirt_front_r}
    componentAppliedLinearAccClamp=<6000 6000 6000>
    componentLinearAccScale=<0.6 0.6 0.6>
    componentLinearVelScale=<0.6 0.6 0.6>
    gravityOverride {z=-980}
    physicsBodyDefinitions [{
      boundBone {boneName=dyn_shirt_front_r}
      boxExtents=<20 20 20>
      constraintSetup {angularLimitsMax=<0 0 20>, angularLimitsMin=<0 -40 0>, angularTarget=<1 0 0>}
      localJointOffset=<5 0 0>
     }]
   }
   n17 = Anim Dynamics - Bone: dyn_shirt_front_l(ComponentPose=@n16, Alpha=1, GravityOverride=@n12, ExternalForce="0,0,0") {
    angularDampingOverride=0.99
    angularSpringConstant=1000
    bAngularSpring=true
    bOverrideAngularDamping=true
    bUseGravityOverride=true
    gravityScale=0.35
    alphaBoolBlend {blendOption=Linear}
    boundBone {boneName=dyn_shirt_front_l}
    componentAppliedLinearAccClamp=<6000 6000 6000>
    componentLinearAccScale=<0.6 0.6 0.6>
    componentLinearVelScale=<0.6 0.6 0.6>
    gravityOverride {z=-980}
    physicsBodyDefinitions [{
      boundBone {boneName=dyn_shirt_front_l}
      boxExtents=<20 20 20>
      constraintSetup {angularLimitsMax=<0 0 0>, angularLimitsMin=<0 -40 -15>, angularTarget=<1 0 0>}
      localJointOffset=<5 0 0>
     }]
   }
   n18 = "Constraint - dyn_necklace_01 (P)"(ComponentPose=@n17, Alpha=Get Target(), ConstraintWeights_0=1) {
    alphaBoolBlend {blendOption=Linear}
    alphaScaleBiasClamp {bInterpResult=true, interpSpeedDecreasing=5, interpSpeedIncreasing=5}
    boneToModify {boneName=dyn_necklace_01}
    constraintSetup [{targetBone {boneName=necklace_collider}}]
    constraintWeights [1]
   }
   Save cached pose 'Necklace'(Pose=Component To Local(ComponentPose=@n18)) {cacheName=Necklace}
   Get bIsAccelerating2D()
   Get bIsMoving2D()
  ]}
 EventGraph {nodes [
   Event BlueprintUpdateAnimation()
   > n1 = "Gravity Override State Function (Nativized)"!(
    execute
    CharacterPartAnimInstance:Custom Character Part Anim Instance Object Reference=Self-Reference()
    GravityOverrideParameterSet:Fort Gravity Override Parameter Set Structure=Get Gravity Override Params_0() -> (Gravity Override Params_0:Fort Gravity Override Parameter Set Structure)
    RigidBodyAnimNodeParameterSet:Fort Rigid Body Anim Node Parameter Set Structure
    DebugDraw:bool=false
    GameTimeInSeconds:float=GetGameTimeInSeconds()
    DebugJointName:Name=root
    StateLogicLayerName:Name=none
   ) -> (
    then
    OutGravityOverride:Vector
    OutCurrentState:Name=None
    OutGravityOverrideParameters:Fort Gravity Override Parameters Structure
    OutRigidBodyAnimNodeParameters:Fort Rigid Body Anim Node Parameters Structure
   )
   > Set GravityOverride(GravityOverride=@n1)
   > Set Parachute(Parachute:double="To Float (Boolean)"(InBool=Get bIsParachuteOpened())) -> (Output_Get:double)
   n2 = Get bIsSprinting()
   > Set Notsprinting(Notsprinting:double="To Float (Boolean)"(InBool=NOT Boolean(A=@n2))) -> (Output_Get:double)
   n3 = Get bIsCrouching()
   > Set crouching(crouching:double="To Float (Boolean)"(InBool=@n3)) -> (Output_Get:double)
   > Set Target(Target:double="To Float (Boolean)"(InBool=Get bIsTargeting())) -> (Output_Get:double)
   > Set Emote(Emote:double="To Float (Boolean)"(InBool=Get bIsPlayingEmote())) -> (Output_Get:double)
   n4 = Get bIsAccelerating2D()
   > Set MoveNOCrouch(MoveNOCrouch:bool=@n4) -> (Output_Get:bool)
   "To Float (Boolean)"(InBool=AND Boolean(A=@n3, B=Get bIsMoving2D()))
   AND Boolean(A=@n4, B=NOT Boolean(A=Get bIsCrouching()))
   Get bIsSkydiveDiveMode()
  ]}
]
variables [
 crouching {type=number}
 Emote {type=number}
 Gravity Override Params_0 {
  type=FortGravityOverrideParameterSet
  default=(Skydive_GoingUp=(GravityOverride=(GravityOverride=(X=0.000000,Y=0.000000,Z=-980.000000))),Skydive_FloatingDown=(GravityOverride=(WindFrequency=(X=0.000000,Y=8.000000,Z=8.000000),WindAmplitude=(X=0.000000,Y=200.000000,Z=200.000000),WindOffset=(X=0.000000,Y=0.000000,Z=-200.000000),JointName="root")),Skydive_DivingDown=(GravityOverride=(WindFrequency=(X=0.000000,Y=20.000000,Z=20.000000),WindAmplitude=(X=0.000000,Y=500.000000,Z=500.000000),WindOffset=(X=0.000000,Y=0.000000,Z=-200.000000))),Skydive_Parachuting=(GravityOverride=(WindFrequency=(X=0.000000,Y=8.000000,Z=8.000000),WindAmplitude=(X=0.000000,Y=200.000000,Z=200.000000),WindOffset=(X=0.000000,Y=0.000000,Z=-200.000000),JointName="root")),OnGround_Standing=(GravityOverride=(GravityOverride=(X=0.000000,Y=0.000000,Z=-980.000000))),OnGround_Moving=(GravityOverride=(GravityOverride=(X=0.000000,Y=0.000000,Z=-980.000000))),OnGround_Moving_HighRotationRate=(GravityOverride=(GravityOverride=(X=0.000000,Y=0.000000,Z=-980.000000))),OnGround_Crouch_Moving=(GravityOverride=(GravityOverride=(X=0.000000,Y=0.000000,Z=-980.000000))),EmoteOrMelee=(GravityOverride=(GravityOverride=(X=0.000000,Y=0.000000,Z=-980.000000))),Swimming=(GravityOverride=(GravityOverride=(X=0.000000,Y=0.000000,Z=-980.000000))),Front_End=(GravityOverride=(GravityOverride=(X=0.000000,Y=0.000000,Z=-980.000000))),DBNO=(GravityOverride=(GravityOverride=(X=0.000000,Y=0.000000,Z=-980.000000))),DBNO_Carried=(GravityOverride=(GravityOverride=(X=0.000000,Y=0.000000,Z=-980.000000))),OnGround_Falling=(GravityOverride=(GravityOverride=(X=0.000000,Y=0.000000,Z=-980.000000))),Ziplining=(GravityOverride=(GravityOverride=(X=0.000000,Y=0.000000,Z=-980.000000))),Skydive_Surfing=(GravityOverride=(GravityOverride=(X=0.000000,Y=0.000000,Z=-980.000000))),Riding_Standing=(GravityOverride=(GravityOverride=(X=0.000000,Y=0.000000,Z=-980.000000))),Riding_Moving=(GravityOverride=(GravityOverride=(X=0.000000,Y=0.000000,Z=-980.000000))),Riding_Moving_Targeting=(GravityOverride=(GravityOverride=(X=0.000000,Y=0.000000,Z=-980.000000))),Tactical_Sprint=(GravityOverride=(GravityOverride=(X=0.000000,Y=0.000000,Z=-980.000000))),Rail_Grinding=(GravityOverride=(GravityOverride=(X=0.000000,Y=0.000000,Z=-980.000000))),)
 }
 GravityOverride {type=Vector}
 MoveNOCrouch {type=boolean}
 Notsprinting {type=number, default=1}
 Parachute {type=number}
 Target {type=number}
]
