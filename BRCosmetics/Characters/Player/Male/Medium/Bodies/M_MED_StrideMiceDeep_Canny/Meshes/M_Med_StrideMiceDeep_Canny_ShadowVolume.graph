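# Physics asset for the StrideMiceDeep Canny male medium player body, named as its shadow volume, likely serving as a simplified capsule shadow caster.
# Single capsules approximate pelvis, upper spine, arms, legs and feet, with no head or hand bodies; spine_03 is only a tiny linking capsule.
# Three extra capsules on shirt bones (back, right, left) cover the garment and constrain directly to the pelvis, while other constraints follow the skeleton chain.
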PhysicsAsset /BRCosmetics/Characters/Player/Male/Medium/Bodies/M_MED_StrideMiceDeep_Canny/Meshes/M_Med_StrideMiceDeep_Canny_ShadowVolume
bodies {
 base_shirt_l [{shapeType=Capsule, length=8.5, radius=7.1, shapeName=base_shirt_l_capsule, center=<4.7 -2 0>, rotation=<rot 0 93 78>}]
 calf_l [{shapeType=Capsule, length=46.1778, radius=8.7701, center=<-19.5359 -1.5212 -0.9683>, rotation=<rot 0 -86.6988 -91.394>}]
 calf_r [{shapeType=Capsule, length=48.8744, radius=9.152, center=<19.3436 1.393 1.158>, rotation=<rot 0 93.3011 -88.6059>}]
 dyn_shirt_back [{
   shapeType=Capsule
   length=14.4215
   radius=8.975
   shapeName=dyn_shirt_back_capsule
   center=<4.1214 -0.1819 0.6386>
   rotation=<rot 0 -178.0958 -89.5801>
  }]
 dyn_shirt_r [{shapeType=Capsule, length=8.5307, radius=7.1053, shapeName=dyn_shirt_r_capsule, center=<4.7 -2 0>, rotation=<rot 0 93 -78>}]
 foot_l [{shapeType=Capsule, length=13.893, radius=9.6701, center=<-3.5938 6.5839 -0.3311>, rotation=<rot -0.1913 -174.5413 -90.2553>}]
 foot_r [{shapeType=Capsule, length=11.393, radius=9.6701, center=<4.2554 -6.483 0.3331>, rotation=<rot 0.1947 4.695 -89.7472>}]
 lowerarm_l [{shapeType=Capsule, length=31.988, radius=5.9446, center=<17.2673 -1.2827 -0.8907>, rotation=<rot 32.332 -90.0001 96.5903>}]
 lowerarm_r [{shapeType=Capsule, length=36.7394, radius=7.4433, center=<-17.428 2.6398 2.3801>, rotation=<rot 32.5381 -89.9999 -83.426>}]
 pelvis [{shapeType=Capsule, length=15.9585, radius=15.4643, center=<3.5228 0.6553 1.7113>, rotation=<rot 5.4333 -89.8581 1.8747>}]
 spine_03 [{shapeType=Capsule, length=0.505, radius=0.505, rotation=<rot 0 85.1284 -90>}]
 spine_04 [{shapeType=Capsule, length=28.6776, radius=17.7595, center=<-2.4936 -0.2829 0.0001>, rotation=<rot 0 90.4359 -90>}]
 thigh_l [{shapeType=Capsule, length=26.1548, radius=10.8417, center=<-17.5075 3.4807 -1.1173>, rotation=<rot -17.5912 -90 -94.5564>}]
 thigh_r [{shapeType=Capsule, length=32.2315, radius=11.2525, center=<17.915 -2.9396 -0.431>, rotation=<rot -9.1473 -90 91.3174>}]
 upperarm_l [{shapeType=Capsule, length=25.4422, radius=6.5842, center=<12.0522 0.3246 1.7238>, rotation=<rot 0 91.7203 -88.8764>}]
 upperarm_r [{shapeType=Capsule, length=24.9422, radius=6.79, center=<-12.0034 -0.8123 -0.7884>, rotation=<rot 0 -88.2798 -86.2691>}]
}
constraints [
 {bone1Name=spine_03, bone2Name=pelvis, swing1LimitDegrees=45, swing2LimitDegrees=45, twistLimitDegrees=45}
 {bone1Name=spine_04, bone2Name=spine_03, swing1LimitDegrees=45, swing2LimitDegrees=45, twistLimitDegrees=45}
 {bone1Name=upperarm_l, bone2Name=spine_04, swing1LimitDegrees=45, swing2LimitDegrees=45, twistLimitDegrees=45}
 {bone1Name=upperarm_r, bone2Name=spine_04, swing1LimitDegrees=45, swing2LimitDegrees=45, twistLimitDegrees=45}
 {bone1Name=calf_l, bone2Name=thigh_l, swing1LimitDegrees=45, swing2LimitDegrees=45, twistLimitDegrees=45}
 {bone1Name=calf_r, bone2Name=thigh_r, swing1LimitDegrees=45, swing2LimitDegrees=45, twistLimitDegrees=45}
 {
  bone1Name=thigh_l
  bone2Name=pelvis
  swing1LimitDegrees=45
  swing1Motion=Free
  swing2LimitDegrees=45
  swing2Motion=Free
  twistLimitDegrees=45
  twistMotion=Free
 }
 {
  bone1Name=thigh_r
  bone2Name=pelvis
  swing1LimitDegrees=45
  swing1Motion=Free
  swing2LimitDegrees=45
  swing2Motion=Free
  twistLimitDegrees=45
  twistMotion=Free
 }
 {
  bone1Name=foot_l
  bone2Name=calf_l
  swing1LimitDegrees=45
  swing1Motion=Free
  swing2LimitDegrees=45
  swing2Motion=Free
  twistLimitDegrees=45
  twistMotion=Free
 }
 {
  bone1Name=foot_r
  bone2Name=calf_r
  swing1LimitDegrees=45
  swing1Motion=Free
  swing2LimitDegrees=45
  swing2Motion=Free
  twistLimitDegrees=45
  twistMotion=Free
 }
 {
  bone1Name=lowerarm_l
  bone2Name=upperarm_l
  swing1LimitDegrees=45
  swing1Motion=Free
  swing2LimitDegrees=45
  swing2Motion=Free
  twistLimitDegrees=45
  twistMotion=Free
 }
 {
  bone1Name=lowerarm_r
  bone2Name=upperarm_r
  swing1LimitDegrees=45
  swing1Motion=Free
  swing2LimitDegrees=45
  swing2Motion=Free
  twistLimitDegrees=45
  twistMotion=Free
 }
 {
  bone1Name=dyn_shirt_back
  bone2Name=pelvis
  swing1LimitDegrees=45
  swing1Motion=Free
  swing2LimitDegrees=45
  swing2Motion=Free
  twistLimitDegrees=45
  twistMotion=Free
 }
 {
  bone1Name=dyn_shirt_r
  bone2Name=pelvis
  swing1LimitDegrees=45
  swing1Motion=Free
  swing2LimitDegrees=45
  swing2Motion=Free
  twistLimitDegrees=45
  twistMotion=Free
 }
 {
  bone1Name=base_shirt_l
  bone2Name=pelvis
  swing1LimitDegrees=45
  swing1Motion=Free
  swing2LimitDegrees=45
  swing2Motion=Free
  twistLimitDegrees=45
  twistMotion=Free
 }
]
